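# Sound cue for the Golden Skeleton pickaxe's impact sound, built from enemy-impact recordings.
# A random node picks one of four equally weighted wave variations on each play.
# It uses the shared pickaxe impact attenuation asset, cosmetic pickaxe property sheet and impacts sound class, and primes on load.
# Concurrency is overridden to a small per-owner limit that stops the oldest sound.
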SoundCue /BRCosmetics/Sounds/PickAxes/GoldenSkeleton/PA_GoldenSkeleton_Impact_Cue
root=Random_0
nodes [
 {id=Random_0, children [WavePlayer_0,WavePlayer_1,WavePlayer_2,WavePlayer_7], properties {weights [1,1,1,1]}}
 {id=WavePlayer_0, properties {soundWaveAssetPtr=&/Pickaxe_GoldenSkeleton_EnemyImpact_01}}
 {id=WavePlayer_1, properties {soundWaveAssetPtr=&/Pickaxe_GoldenSkeleton_EnemyImpact_02}}
 {id=WavePlayer_2, properties {soundWaveAssetPtr=&/Pickaxe_GoldenSkeleton_EnemyImpact_04}}
 {id=WavePlayer_7, properties {soundWaveAssetPtr=&/Pickaxe_GoldenSkeleton_EnemyImpact_06}}
]
properties {
 attenuationSettings=/Game/Sounds/Attenuation/Fort_Pickaxe_Impact_Att
 audioPropertiesSheet=/BRCosmetics/Sounds/Settings/PropertySheets/Pickaxe/APS_Cosmetic_Pickaxe_Impact
 bOverrideConcurrency=true
 bPrimeOnLoad=true
 maxDistance=4300
 soundClassObject=&Fort_Impacts
 volumeMultiplier=0.9
 attenuationOverrides {
  bAttenuateWithLPF=true
  distanceAlgorithm=Logarithmic
  falloffDistance=800
  lPFRadiusMax=800
  lPFRadiusMin=400
  attenuationShapeExtents {x=300}
 }
 concurrencyOverrides {bLimitToOwner=true, maxCount=2, resolutionRule=StopOldest}
 concurrencySet [&Player_Melee_Weapon_Conc]
}
